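; Golf bag cart prop Blueprint that applies color or material variations to its SM_Recreation_GolfBag_A mesh at construction.
; After the parent construction script runs, it looks up the first building texture data entry in the TDMats map.
; With a match, swatch index 0 tints the bag with that data's tint color; other indices set its override material on slot 0, if valid.
; BeginPlay, overlap and Tick events are empty.
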(event EventBeginPlay)

(event Collision|EventActorBeginOverlap
  (params (OtherActor :type Actor%20Object%20Reference)))

(event EventTick
  (params (DeltaSeconds :type Float%20%28single-precision%29)))

(fn ConstructionScript
  (bind _map_tdmats (Variables|Components|GetMap_TDMats))
  (|CallParentFunction)
  (bind _outtexturedataarray (Building|GetBuildingTextureData))
  (if (Utilities|Array|IsNotEmpty _outtexturedataarray)
    (if (.ReturnValue (Utilities|Map|Find _map_tdmats (Utilities|Array|Get(acopy) _outtexturedataarray)))
      (if (== (Variables|RandomSwatch|GetSwatchIndex) 0)
        (Default|SetColoronSecondaryMesh (Math|Conversions|ToLinearColor(Color) (Variables|/Script/FortniteGame.BuildingTextureData|GetTintColor (Utilities|Map|Find _map_tdmats (Utilities|Array|Get(acopy) _outtexturedataarray)))) (Variables|Default|GetSM_Recreation_GolfBag_A))
        (else
          (Utilities|IsValid (Variables|/Script/FortniteGame.BuildingTextureData|GetOverrideMaterial (Utilities|Map|Find _map_tdmats (Utilities|Array|Get(acopy) _outtexturedataarray)))
            (:Is%20Valid
              (Rendering|Material|SetMaterial (Variables|Default|GetSM_Recreation_GolfBag_A) 0 (Variables|/Script/FortniteGame.BuildingTextureData|GetOverrideMaterial (Utilities|Map|Find _map_tdmats (Utilities|Array|Get(acopy) _outtexturedataarray)))))))))))
